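# Editor utility function library whose DiffTextures function reports whether two Texture2D assets match, with a reason string.
# It compares width, height and source format in turn, setting a specific reason on the first mismatch.
# The format's index in TypeLookup selects an FColor or FLinearColor read of mip 0 source pixels, compared as whole arrays.
# Any other lookup index returns a match without comparing pixels.
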EditorUtilityBlueprint /TechArtTools/FunctionLibrary/EUFL_TechArtToolsBlueprintFunctionLibrary
graphs [DiffTextures {nodes [
   n1 = DiffTextures!() -> (then, Texture A:Texture 2D Object Reference, Texture B:Texture 2D Object Reference) {is=title K2Node_FunctionEntry}
   > Set LocalBool(LocalBool:bool=true) -> (Output_Get:bool)
   > Set LocalReason(LocalReason:String="Match!") -> (Output_Get:String)
   > n2 = Branch(Condition="Equal (Integer)"(A=GetSizeX!(self:Texture 2D Object Reference=@n1) -> (ReturnValue:int=0), B=GetSizeX(self=@n1.Texture B) -> (ReturnValue=0)))
   > n3 = Branch(Condition="Equal (Integer)"(A=GetSizeY!(self:Texture 2D Object Reference=@n1) -> (ReturnValue:int=0), B=GetSizeY(self=@n1.Texture B) -> (ReturnValue=0)))
   n4 = Get Texture2D Source Format!(Texture:Texture 2D Object Reference=@n1.Texture B, LayerIndex:int=0) -> (ReturnValue:String) {is=title K2Node_CallFunction}
   > n5 = Branch(Condition="Equal, Case Insensitive (String)"(A=Get Texture2D Source Format(Texture=@n1, LayerIndex=0), B=@n4))
   > n6 = Switch on Int!(
    execute:Exec
    Selection:int=Find Item!(TargetArray=Get TypeLookup() -> (TypeLookup:Array of Strings), ItemToFind:"String (by ref)"=@n4) -> (ReturnValue)
   ) -> (Default:Exec, 0:Exec, 1:Exec)
   n7 = "Read Texture2D Source Pixels (FColor)"!(execute=@n6.0, Texture:Texture 2D Object Reference=@n1, MipIndex:int=0) -> (then, OutPixels:Array of Color Structures, OutWidth:int=0, OutHeight:int=0, OutErrorMessage:String, ReturnValue:bool)
   > n8 = "Read Texture2D Source Pixels (FColor)"(Texture=@n1.Texture B, MipIndex=0) -> (OutWidth=0, OutHeight=0)
   > n9 = Branch(Condition=Identical!(ArrayA:Array of Color Structures=@n7, ArrayB:Array of Color Structures=@n8) -> (ReturnValue:bool) {is=title K2Node_CallArrayFunction})
   Return Node!(
    execute=[@n13.then, @n9.then, @n6.Default, @n10.then, @n14.then, @n17.then, @n16.then, @n15.then]
    TextureMatch:bool=Get LocalBool() -> (LocalBool:bool)
    Reason:String=Get LocalReason() -> (LocalReason:String)
   )
   Set LocalBool(execute=@n9.else, LocalBool=false)
   > n10 = Set LocalReason(LocalReason=Pixels aren't equal)
   n11 = "Read Texture2D Source Pixels (FLinearColor)"!(execute=@n6.1, Texture:Texture 2D Object Reference=@n1, MipIndex:int=0) -> (then, OutPixels:Array of Linear Color Structures, OutWidth:int=0, OutHeight:int=0, OutErrorMessage:String, ReturnValue:bool)
   > n12 = "Read Texture2D Source Pixels (FLinearColor)"(Texture=@n1.Texture B, MipIndex=0) -> (OutWidth=0, OutHeight=0)
   > n13 = Branch(Condition=Identical(ArrayA:Array of Linear Color Structures=@n11, ArrayB:Array of Linear Color Structures=@n12))
   Set LocalBool(execute=@n13.else, LocalBool=false)
   > n14 = Set LocalReason(LocalReason=Pixels aren't equal)
   Set LocalBool(execute=@n5.else, LocalBool=false)
   > n15 = Set LocalReason(LocalReason=Source format is different)
   Set LocalBool(execute=@n3.else, LocalBool=false)
   > n16 = Set LocalReason(LocalReason=Y Dimensions Don't Match)
   Set LocalBool(execute=@n2.else, LocalBool=false)
   > n17 = Set LocalReason(LocalReason=X Dimensions Don't Match)
  ]}]
variables []
